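; Dance variant of the POI banner actor that adds Niagara effects for capture and dancing, plus sky fireworks when the banner is raised.
; BindVFXEvents hooks the first capture point's state changes to start or stop the dance effect, and its accelerator begin/end events to toggle "IsDancing".
; Effect changes go through RPC events.
; The HandleBannerRaised override calls ToggleFX, noted as needing an RPC for fireworks to show in PIE.
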(event Custom|ToggleFX
  ; For some reason this only works through an RPC call in PIE otherwise the FXs won't be visible.
  (Components|Activation|Activate (Variables|Default|GetSkyFireworks) true))

(event Custom|BindVFXEvents
  (CallFunction|RPCVFXStoppedDancing)
  (bind _capturepoint (CallFunction|GetFirstProgressSourceasCapturePoint))
  (EventDispatchers|BindEventtoOnCaptureStateChanged _capturepoint (delegate Custom|CaptureStateChanged))
  (Default|BindEventtoOnAccelerationBegin (Variables|/POIBanner/Actors/CapturePoint/BGA_POIBanner_CapturePoint.BGA_POIBanner_CapturePoint_C|GetAccelerators _capturepoint) (delegate Custom|DancingStarted))
  (Default|BindEventtoOnAccelerationEnd (Variables|/POIBanner/Actors/CapturePoint/BGA_POIBanner_CapturePoint.BGA_POIBanner_CapturePoint_C|GetAccelerators _capturepoint) (delegate Custom|DancingEnded)))

(event Custom|CaptureStateChanged
  (params (CapturePoint :type BGA%20POIBanner%20Capture%20Point%20Object%20Reference) (NewState :type EPOIBannerState%20Enum :default "NewEnumerator0") (OldState :type EPOIBannerState%20Enum :default "NewEnumerator0") (LastCapturingTeam :type Integer))
  (switch Utilities|FlowControl|Switch|SwitchonEPOIBannerState NewState
    (:NewEnumerator0
      _)
    (:NewEnumerator1
      (CallFunction|RPCVFXStartedCapture))
    (:NewEnumerator2
      _)
    (:NewEnumerator3
      _))
  (CallFunction|RPCVFXStoppedCapture))

(event Custom|DancingStarted
  (params (Team :type Integer))
  (CallFunction|RPCVFXStartedDancing))

(event Custom|DancingEnded
  (params (Team :type Integer))
  (CallFunction|RPCVFXStoppedDancing))

(event Custom|RPC_VFX_StartedDancing
  (Niagara|SetNiagaraVariable(Bool) (Variables|Default|GetDanceActiveFX) "IsDancing" true))

(event Custom|RPC_VFX_StartedCapture
  (Components|Activation|Activate (Variables|Default|GetDanceActiveFX) false))

(event Custom|RPC_VFX_StoppedDancing
  (Niagara|SetNiagaraVariable(Bool) (Variables|Default|GetDanceActiveFX) "IsDancing" false))

(event Custom|RPC_VFX_StoppedCapture
  (Components|Activation|Deactivate (Variables|Default|GetDanceActiveFX)))

(fn HandleBannerRaised
  (params (Team :type Integer))
  (|CallParentFunction Team)
  (CallFunction|ToggleFX))

(fn ConstructionScript
  (|CallParentFunction))
