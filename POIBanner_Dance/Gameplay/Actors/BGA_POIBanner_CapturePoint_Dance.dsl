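; Blueprint actor for a capture-point POI banner, apparently the Dance variant going by its name and path.
; It overrides GetProgressAccelerationScalar to return the parent's result unchanged, and its construction script only calls the parent.
; The graphs add no logic of their own, so behavior in both comes entirely from the parent class.
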(fn GetProgressAccelerationScalar
  (return (|CallParentFunction)))

(fn ConstructionScript
  (|CallParentFunction))
