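; Capture-point accelerator actor that speeds up point-of-interest capture when player pawns emote inside its DanceCollider.
; On authority, overlap tracks pawns in the zone and binds or unbinds their emote start/stop events.
; Each newly emoting pawn begins acceleration for its team, which ends only when the emoting list empties, without an alliance check as enemies halt progress anyway.
; Emote changes are broadcast cosmetically for the progression widget's dance prompt.
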(event OnComponentBeginOverlap(DanceCollider)
  (params (OverlappedComponent :type Primitive%20Component%20Object%20Reference) (OtherActor :type Actor%20Object%20Reference) (OtherComp :type Primitive%20Component%20Object%20Reference) (OtherBodyIndex :type Integer) (bFromSweep :type Boolean) (SweepResult :type Hit%20Result%20Structure%20%28by%20ref%29))
  (Networking|SwitchHasAuthority
    (:Authority
      (bind _asfort_player_pawn (Utilities|Casting|CastToFortPlayerPawn OtherActor))
      (Utilities|Array|Add (Variables|Default|GetPawnsInZone) _asfort_player_pawn)
      (EventDispatchers|BindEventtoOnPawnStartedEmote _asfort_player_pawn (EventDispatchers|CreateEvent :!event Custom|StartedEmote))
      (EventDispatchers|BindEventtoOnPawnStoppedEmote _asfort_player_pawn (EventDispatchers|CreateEvent :!event Custom|StoppedEmote)))))

(event OnComponentEndOverlap(DanceCollider)
  (params (OverlappedComponent :type Primitive%20Component%20Object%20Reference) (OtherActor :type Actor%20Object%20Reference) (OtherComp :type Primitive%20Component%20Object%20Reference) (OtherBodyIndex :type Integer))
  (Networking|SwitchHasAuthority
    (:Authority
      (bind _asfort_player_pawn (Utilities|Casting|CastToFortPlayerPawn OtherActor))
      (Utilities|Array|RemoveItem (Variables|Default|GetPawnsInZone) _asfort_player_pawn)
      (EventDispatchers|UnbindEventfromOnPawnStartedEmote _asfort_player_pawn (EventDispatchers|CreateEvent :!event Custom|StartedEmote))
      (EventDispatchers|UnbindEventfromOnPawnStoppedEmote _asfort_player_pawn (EventDispatchers|CreateEvent :!event Custom|StoppedEmote)))))

(event Custom|StartedEmote
  (params (MontageItemDef :type Fort%20Item%20Definition%20Object%20Reference) (PawnEmoting :type Fort%20Pawn%20Object%20Reference))
  (bind _pawnsemoting (Variables|Default|GetPawnsEmoting))
  (bind _asfort_player_pawn (Utilities|Casting|CastToFortPlayerPawn PawnEmoting))
  (bind _output_get (Variables|Default|SetAs%20Fort%20Player%20Pawn _asfort_player_pawn))
  (if (not (Utilities|Array|ContainsItem _pawnsemoting _output_get))
    (Utilities|Array|Add _pawnsemoting _output_get)
    (CallFunction|BeginAcceleration (Teams|TeamtoIndex (Team|GetActorTeam _output_get)))
    (CallFunction|OnNewPawnEmotingUpdated (Variables|Default|GetAs%20Fort%20Player%20Pawn) true)))

(event Custom|StoppedEmote
  (params (MontageItemDef :type Fort%20Item%20Definition%20Object%20Reference) (PawnEmoting :type Fort%20Pawn%20Object%20Reference))
  (bind _pawnsemoting (Variables|Default|GetPawnsEmoting))
  (bind _asfort_player_pawn (Utilities|Casting|CastToFortPlayerPawn PawnEmoting))
  (Utilities|FlowControl|Sequence
    (:then_0
      (Utilities|Array|RemoveItem _pawnsemoting _asfort_player_pawn)
      ;[0.15 0.3 0.15 0.5] Little bit of a hack but we don't need to check for team alliegance here because enemies in zone stops progress regardless
      (if (<= (Utilities|Array|Length _pawnsemoting) 0)
        (CallFunction|EndAcceleration (Teams|TeamtoIndex (Team|GetActorTeam _asfort_player_pawn)))))
    (:then_1
      (CallFunction|OnNewPawnEmotingUpdated (Variables|Default|GetAs%20Fort%20Player%20Pawn)))))

(event Custom|OnNewPawnEmotingUpdated
  (params (FortPlayerPawn :type Fort%20Player%20Pawn%20Object%20Reference) (IsEmoting :type Boolean))
  ; UI - Update Progression Widget Dance Prompt
  (Networking|CanExecuteCosmeticEvents
    (:True
      (Default|CallOnPawnEmotingUpdated FortPlayerPawn IsEmoting))))

(fn ConstructionScript
  (|CallParentFunction))
