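; Dance variant of the POI claim spawner, which adds an accelerator to the POI banner capture point setup.
; Its control point spawn override runs the parent, spawns a Dance accelerator at the given location, links it to every control point, initializes listeners and binds the Dance banner's VFX events.
; Once playlist data is ready, it destroys the level-design floor mesh behind an editor-or-PIE check.
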(event EventBeginPlay
  (|CallParentFunction)
  (Class|FortAsyncActionCurrentPlaylistReady|PlaylistDataReadyAsync
    (:OnReady
      ;[0.5 0.4 0.15 0.5] These components are visual representation for level design when placing spawners, but we don't want them in PIE or cooked
      (if (Editor|InEditororPIE)
        (Components|DestroyComponent (Variables|Default|GetSM_Floor_Base_1))))))

(fn SpawnControlPointActors
  (params (Location :type Vector :default "0, 0, 0"))
  (|CallParentFunction Location)
  (bind _returnvalue (Game|SpawnActorfromClass "/POIBanner_Dance/Gameplay/Actors/BGA_POIWin_CapturePoint_Accelerator_Dance.BGA_POIWin_CapturePoint_Accelerator_Dance_C" (Math|Transform|MakeTransform Location) "AlwaysSpawn" "SelectDefaultAtRuntime" self 0 0 0))
  (for _array_element (Variables|SpawnedActors|GetControlPoints)
    (Variables|/POIBanner/Actors/CapturePoint/BGA_POIBanner_CapturePoint.BGA_POIBanner_CapturePoint_C|SetAccelerators (Utilities|Array|MakeArray _returnvalue) _array_element)
    (Acceleration|InitializeAcceleratorListeners _array_element)
    (bind _asbga_poibanner_banner_dance (Utilities|Casting|CastToBGA_POIBanner_Banner_Dance (Variables|SpawnedActors|GetBanner)))
    (Class|BGAPOIBannerBannerDance|BindVFXEvents _asbga_poibanner_banner_dance)))

(fn ConstructionScript
  (|CallParentFunction))
